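# Physics asset for the CreamSkull guitar cosmetic, giving physics bodies to a three-bone dynamic cardstock piece attached to its handle.
# A capsule on the handle anchors a serial chain of three box bodies, each constrained to the one before it.
# All joints lock twist, and the two outer joints also lock one swing axis, so they bend hinge-like on a single axis.
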PhysicsAsset /SparksCosmetics/Instrument/Guitar/CreamSkull/Meshes/CreamSkull_Guitar_Physics
bodies {
 dyn_cardstock_01 [{
   shapeType=Box
   extentX=0.3331
   extentY=2.5815
   extentZ=5.9883
   shapeName=dyn_cardstock_01_box
   center=<1.3362 -0.0176 -0.1782>
   rotation=<rot -14.2387 -90 4.7598>
  }]
 dyn_cardstock_02 [{
   shapeType=Box
   extentX=0.3508
   extentY=1.395
   extentZ=5.7372
   shapeName=dyn_cardstock_02_box
   center=<0.634 0.008 -0.1587>
   rotation=<rot -6.2758 -90 -3.0462>
  }]
 dyn_cardstock_03 [{
   shapeType=Box
   extentX=1.01
   extentY=1.01
   extentZ=1.01
   shapeName=dyn_cardstock_03_box
   center=<0.7818 0.1437 -0.0409>
   rotation=<rot -6.288 -90 -3.7262>
  }]
 handle [{shapeType=Capsule, length=10, radius=2}]
}
constraints [
 {bone1Name=dyn_cardstock_01, bone2Name=handle, swing1LimitDegrees=4, swing2LimitDegrees=4, twistLimitDegrees=45, twistMotion=Locked}
 {
  bone1Name=dyn_cardstock_02
  bone2Name=dyn_cardstock_01
  swing1LimitDegrees=8
  swing2LimitDegrees=2
  swing2Motion=Locked
  twistLimitDegrees=45
  twistMotion=Locked
 }
 {
  bone1Name=dyn_cardstock_03
  bone2Name=dyn_cardstock_02
  swing1LimitDegrees=6
  swing2LimitDegrees=2
  swing2Motion=Locked
  twistLimitDegrees=45
  twistMotion=Locked
 }
]
